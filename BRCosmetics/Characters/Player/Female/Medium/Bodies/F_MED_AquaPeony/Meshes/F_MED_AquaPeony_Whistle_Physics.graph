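# Physics asset for the whistle on the F_MED_AquaPeony female character body, setting up a whistle hanging on a string against the torso.
# Capsules on spine_04, spine_05 and both pectoral deform bones form the body shape, joined by zero-limit constraints.
# A chain runs from a master bone pinned to spine_05 through three string segments to the whistle, with limited swing and little or no twist.
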PhysicsAsset /BRCosmetics/Characters/Player/Female/Medium/Bodies/F_MED_AquaPeony/Meshes/F_MED_AquaPeony_Whistle_Physics
bodies {
 deform_pec_base_l [{shapeType=Capsule, length=5, radius=5, shapeName=deform_pec_l_capsule, center=<-4.7698 -0.2155 -2.1527>, rotation=<rot -56.6486 0 -2.6759>}]
 deform_pec_base_r [{shapeType=Capsule, length=5, radius=5, shapeName=deform_pec_l_capsule, center=<4.7698 0.2155 2.1527>, rotation=<rot -56.6486 0 -2.6759>}]
 dyn_whistle [{
   shapeType=Capsule
   length=3.2724
   radius=1.5454
   shapeName=dyn_whistle_capsule
   center=<2.3855 -2.1243 -0.0004>
   rotation=<rot -0.0038 -104.6527 -89.9818>
  }]
 dyn_whistle_string_01 [{shapeType=Capsule, length=3, radius=1.5, shapeName=dyn_whistle_string_01_capsule, center=<1 0 0>}]
 dyn_whistle_string_02 [{shapeType=Capsule, length=3, radius=1.5, shapeName=dyn_whistle_string_02_capsule, center=<1 -0.7 0>}]
 dyn_whistle_string_03 [{shapeType=Capsule, length=3, radius=1.5, shapeName=dyn_whistle_string_03_capsule, center=<1 -1 0>}]
 dyn_whistle_string_MSTR [{shapeType=Capsule, length=0.5, radius=0.5, shapeName=dyn_whistle_string_MSTR_capsule}]
 spine_04 [
  {shapeType=Capsule, length=5.4, radius=5.4, shapeName=spine_04_capsule_l, center=<-0.1171 8.417 -4.0572>}
  {shapeType=Capsule, length=5.4, radius=5.4, shapeName=spine_04_capsule_l, center=<-0.1171 8.417 4.0572>}
  {shapeType=Capsule, length=15, radius=6.8, shapeName=spine_04_capsule, center=<3.1228 4.0995 0>}
 ]
 spine_05 [{shapeType=Capsule, length=9, radius=10, shapeName=spine_05_capsule, center=<-2.9489 -0.2376 0>}]
}
constraints [
 {bone1Name=spine_05, bone2Name=spine_04, swing1LimitDegrees=0, swing2LimitDegrees=0, twistLimitDegrees=0}
 {bone1Name=dyn_whistle_string_MSTR, bone2Name=spine_05, swing1LimitDegrees=0, swing2LimitDegrees=0, twistLimitDegrees=0}
 {bone1Name=dyn_whistle_string_01, bone2Name=dyn_whistle_string_MSTR, swing1LimitDegrees=20, swing2LimitDegrees=10, twistLimitDegrees=1}
 {bone1Name=dyn_whistle_string_02, bone2Name=dyn_whistle_string_01, swing1LimitDegrees=30, swing2LimitDegrees=30, twistLimitDegrees=1}
 {bone1Name=dyn_whistle_string_03, bone2Name=dyn_whistle_string_02, swing1LimitDegrees=30, swing2LimitDegrees=30, twistLimitDegrees=1}
 {bone1Name=dyn_whistle, bone2Name=dyn_whistle_string_03, swing1LimitDegrees=30, swing2LimitDegrees=0, twistLimitDegrees=0}
 {bone1Name=deform_pec_base_r, bone2Name=spine_04, swing1LimitDegrees=0, swing2LimitDegrees=0, twistLimitDegrees=0}
 {bone1Name=deform_pec_base_l, bone2Name=spine_04, swing1LimitDegrees=0, swing2LimitDegrees=0, twistLimitDegrees=0}
]
