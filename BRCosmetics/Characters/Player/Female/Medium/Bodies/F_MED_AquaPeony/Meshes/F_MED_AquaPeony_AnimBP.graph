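# Post-process-style AnimBP for the F_MED_AquaPeony female body cosmetic, adding secondary motion to its bag and whistle.
# Copies the attached parent's pose, then runs AnimDynamics on dyn_bag, the body's Control Rig, and three chained RigidBody passes in spine_05 space with differing gravity settings.
# In the front end, the owner's Frontend_Idle_Bypass curve weights a DefaultSlot layered blend from deform_bag_MSTR, letting montages override the bag.
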AnimBlueprint /BRCosmetics/Characters/Player/Female/Medium/Bodies/F_MED_AquaPeony/Meshes/F_MED_AquaPeony_AnimBP
target_skeleton=&/F_MED_AquaPeony_Skeleton
graphs [
 AnimGraph {nodes [
   n1 = Anim Dynamics - Bone: dyn_bag(ComponentPose=Local To Component(LocalPose=Copy Pose From Mesh() {bUseAttachedParent=true}), Alpha=1, ExternalForce="0,0,0") {
    angularDampingOverride=0.99
    bOverrideAngularDamping=true
    bUseGravityOverride=true
    simulationSpace=RootRelative
    alphaBoolBlend {blendOption=Linear}
    boundBone {boneName=dyn_bag}
    componentAppliedLinearAccClamp=<300 300 1000>
    componentLinearAccScale=<0.2 0.2 0.7>
    componentLinearVelScale=<0.3 0.3 0.8>
    gravityOverride {z=-980}
    physicsBodyDefinitions [{
      boundBone {boneName=dyn_bag}
      boxExtents=<10 10 10>
      constraintSetup {angularLimitsMax=<0 0 0>, angularLimitsMin=<0 -50 0>, angularTarget=<0 0 0>}
      localJointOffset=<1 0 0>
     }]
   }
   n2 = Control Rig(Source=Component To Local(ComponentPose=@n1)) {
    alphaBoolBlend {blendOption=Linear}
    controlRigAssetReference {blueprintRigClass=&/F_MED_AquaPeony_CtrlRig.F_MED_AquaPeony_CtrlRig_C}
   }
   n3 = RigidBody(ComponentPose=Local To Component(LocalPose=@n2)) {
    bClampLinearTranslationLimitToRefPose=true
    bOverrideWorldGravity=true
    overridePhysicsAsset=&/F_MED_AquaPeony_Whistle_Physics
    simulationSpace=BaseBoneSpace
    alphaBoolBlend {blendOption=Linear}
    baseBoneRef {boneName=spine_05}
    overrideWorldGravity {z=-980}
    simSpaceSettings {
     dampingAlpha=0
     maxAngularAcceleration=40
     maxAngularVelocity=20
     maxLinearAcceleration=1100
     maxLinearVelocity=700
     worldAlpha=1
     externalLinearDragV {z=0.5}
    }
   }
   n4 = RigidBody!(ComponentPose=@n3, bAlphaBoolEnabled:bool, ExternalForce) -> (Pose) {
    alphaInputType=Bool
    bClampLinearTranslationLimitToRefPose=true
    bOverrideWorldGravity=true
    overridePhysicsAsset=&F_MED_AquaPeony_Whistle_Physics
    simulationSpace=BaseBoneSpace
    alphaBoolBlend {blendInTime=0.2, blendOption=Cubic, blendOutTime=0.2}
    baseBoneRef {boneName=spine_05}
    overrideWorldGravity {z=200}
    simSpaceSettings {
     dampingAlpha=0
     maxAngularAcceleration=40
     maxAngularVelocity=20
     maxLinearAcceleration=1100
     maxLinearVelocity=700
     worldAlpha=1
     externalLinearDragV {z=0.5}
    }
   }
   n5 = RigidBody(ComponentPose=@n4) {
    alphaInputType=Bool
    bClampLinearTranslationLimitToRefPose=true
    bOverrideWorldGravity=true
    overridePhysicsAsset=&F_MED_AquaPeony_Whistle_Physics
    simulationSpace=BaseBoneSpace
    alphaBoolBlend {blendInTime=0.2, blendOption=Cubic, blendOutTime=0.2}
    baseBoneRef {boneName=spine_05}
    simSpaceSettings {
     dampingAlpha=0
     maxAngularAcceleration=40
     maxAngularVelocity=20
     maxLinearAcceleration=1100
     maxLinearVelocity=700
     worldAlpha=1
     externalLinearDragV {z=0.5}
    }
   }
   Output Pose(Result=Layered blend per bone(
     BasePose=Component To Local(ComponentPose=@n5)
     BlendPoses_0=Slot 'DefaultSlot'()
     BlendWeights_0=Get Custom Idle is Playing() -> (Custom Idle is Playing:double)
    ) {layerSetup [{branchFilters [{boneName=deform_bag_MSTR}]}]}) {layerGroup=None}
  ]}
 EventGraph {nodes [
   Event BlueprintUpdateAnimation()
   > n1 = Sequence()
   Branch(execute=@n1.then_1, Condition=Get bIsInFrontEnd())
   > Set Custom Idle is Playing(Custom Idle is Playing:double=GetCurveValue(self=GetOwnerFrontEndAnimInstance(), CurveName=Frontend_Idle_Bypass)) -> (Output_Get:double)
  ]}
]
variables [Custom Idle is Playing {type=number}]
