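# Control Rig for the AquaPeony female character's face accessory part, applying a corrective pose to the hair ponytail when a backpack is worn.
# A spherical pose reader measures head orientation relative to spine_05 within an asymmetric active and falloff region.
# Its output is multiplied by the isWearingBackpack double (default 0), which gates the correction.
# The result weights three chained ModifyTransforms that offset the first three ponytail bones, including squashing the first.
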ControlRigBlueprint /BRCosmetics/Characters/Player/Female/Medium/Bodies/F_MED_AquaPeony/Meshes/Parts/F_MED_AquaPeony_FaceAcc_CtrlRig
graphs [
 RigVMFunctionLibrary {links [], nodes []}
 RigVMModel {nodes [
   RigUnit_BeginExecution = &RigUnit_BeginExecution()
   n1 = &RigUnit_SphericalPoseReader(
    ActiveRegionScaleFactors=(PositiveWidth=1.2,NegativeWidth=1,PositiveHeight=2.3,NegativeHeight=1.8)
    ActiveRegionSize=0.1
    Debug=(bDrawDebug=false,bDraw2D=False,bDrawLocalAxes=False,DebugScale=25,DebugSegments=20,DebugThickness=0.25)
    DriverAxis=(X=0,Y=1,Z=0)
    DriverItem=(Type=Bone,Name="head")
    FalloffRegionScaleFactors=(PositiveWidth=2,NegativeWidth=1.5,PositiveHeight=7,NegativeHeight=3)
    FalloffSize=0.25
    FlipHeightScaling=False
    FlipWidthScaling=False
    OptionalParentItem=(Type=Bone,Name="spine_05")
    RotationOffset=(X=87,Y=0,Z=-60)
    ExecutePin=@RigUnit_BeginExecution.ExecutePin
   )
   n2 = &RigVMFunction_MathDoubleMul(A=@n1.OutputParam, B=RigVMVariableNode() {variable=isWearingBackpack}.Value)
   > &RigUnit_ModifyTransforms(
    ItemToModify=((Item=(Type=Bone,Name="dyn_ponytail_00"),Transform=(Rotation=(X=0,Y=0,Z=-0.1736,W=0.9848),Translation=(X=0,Y=0,Z=-1.3),Scale3D=(X=0.8,Y=1,Z=1))))
    WeightMinimum=0
    Weight=@n2.Result
   )
   > &RigUnit_ModifyTransforms(
    ItemToModify=((Item=(Type=Bone,Name="dyn_ponytail_01"),Transform=(Rotation=(X=-0.0298,Y=-0.0819,Z=-0.3407,W=0.9361),Translation=(X=2,Y=0,Z=0),Scale3D=(X=1,Y=1,Z=1))))
    WeightMinimum=0
    Weight=@n2.Result
   )
   > &RigUnit_ModifyTransforms(
    ItemToModify=((Item=(Type=Bone,Name="dyn_ponytail_02"),Transform=(Rotation=(X=0.0057,Y=0.1304,Z=-0.0432,W=0.9905),Translation=(X=0,Y=0,Z=0),Scale3D=(X=1,Y=1,Z=1))))
    WeightMinimum=0
    Weight=@n2.Result
   )
  ]}
]
hierarchy [
 root {loc=<0 0 0>, rot=<0 0 0 1>}
 pelvis {parent=root, loc=<0 -0.5301 98.0571>, rot=<0.0127 0.707 -0.0127 -0.707>}
 spine_01 {parent=pelvis, loc=<6.1208 -0.9328 0>, rot=<0 0 0.034 -0.9994>}
 spine_02 {parent=spine_01, loc=<6.5997 -0.2095 0>, rot=<0 0 0.0238 -0.9997>}
 spine_03 {parent=spine_02, loc=<6.4914 -0.1237 0>, rot=<0 0 0.0149 -0.9999>}
 spine_04 {parent=spine_03, loc=<6.9598 0.0794 0>, rot=<0 0 -0.0132 -0.9999>}
 spine_05 {parent=spine_04, loc=<6.9987 0 0>, rot=<0 0 -0.0015 -1>}
 neck_01 {parent=spine_05, loc=<11.3138 0.1529 0>, rot=<0 0 -0.2003 -0.9797>}
 neck_02 {parent=neck_01, loc=<4.5627 0 0>, rot=<0 0 0.0504 -0.9987>}
 head {parent=neck_02, loc=<5.3914 0 0>, rot=<0 0 0.115 -0.9934>}
 dyn_bangs_D_00 {parent=head, loc=<10.9392 9.78 -0.7552>, rot=<-0.1643 0.6382 0.6896 0.3002>}
 dyn_bangs_D_01 {parent=dyn_bangs_D_00, loc=<3.4859 0 0>, rot=<0 0 0 1>}
 dyn_bangs_C_00 {parent=head, loc=<9.3221 10.2937 1.4749>, rot=<-0.1834 0.6746 0.6845 0.2068>}
 dyn_bangs_C_01 {parent=dyn_bangs_C_00, loc=<2.3863 0 0>, rot=<0.0013 -0.2686 0.0849 -0.9595>}
 dyn_bangs_B_00 {parent=head, loc=<8.6095 10.0364 3.6944>, rot=<-0.1935 0.6677 0.6818 0.2276>}
 dyn_bangs_A_00 {parent=head, loc=<8.3982 7.588 6.1543>, rot=<-0.1263 -0.7348 -0.6599 0.093>}
 dyn_bangs_A_01 {parent=dyn_bangs_A_00, loc=<2.5382 -0.2396 -0.0394>, rot=<-0.1307 -0.0168 0.2928 -0.9471>}
 dyn_ponytail_base {parent=head, loc=<10.842 -7.2756 0>, rot=<0.4828 -0.5166 0.5166 -0.4828>}
 dyn_ponytail_00 {parent=dyn_ponytail_base, loc=<0 0 0>, rot=<0 0 0 1>}
 dyn_ponytail_01 {parent=dyn_ponytail_00, loc=<5.6893 0 0>, rot=<0 -0.6106 0 -0.7919>}
 dyn_ponytail_02 {parent=dyn_ponytail_01, loc=<5.5655 0 0>, rot=<0.0024 -0.3702 0.0061 -0.9289>}
 dyn_ponytail_03 {parent=dyn_ponytail_02, loc=<5.5514 0 0>, rot=<0 0 0 1>}
 dyn_strand_E {parent=dyn_ponytail_02, loc=<3.1949 1.3533 2.3739>, rot=<-0.1327 -0.0032 -0.0933 -0.9867>}
 dyn_strand_C {parent=dyn_ponytail_01, loc=<-0.2611 -1.9488 3.5711>, rot=<-0.1228 0.0892 0.0341 -0.9878>}
 dyn_strand_B {parent=dyn_ponytail_01, loc=<2.4166 4.4449 -0.4369>, rot=<0.8073 -0.0035 -0.1428 -0.5726>}
 dyn_strand_A {parent=dyn_ponytail_01, loc=<1.6204 -2.9902 -1.5234>, rot=<-0.3219 -0.1741 0.273 -0.8897>}
 dyn_L_sidepiece_00 {parent=head, loc=<7.4905 6.0636 -6.8959>, rot=<0.0466 -0.9963 0.072 0.0028>}
 dyn_L_sidepiece_01 {parent=dyn_L_sidepiece_00, loc=<6.9258 0 0>, rot=<0 0 0 1>}
 dyn_R_sidepiece_00 {parent=head, loc=<5.5519 2.981 7.6733>, rot=<0.0499 -0.7311 -0.6792 -0.0407>}
 dyn_R_sidepiece_01 {parent=dyn_R_sidepiece_00, loc=<5.4871 0 0>, rot=<0.0391 0.0573 0.051 -0.9963>}
]
variables [isWearingBackpack {type=double, default=0}]
