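# Physics asset for the AquaPeony female character's face accessory mesh, defining a ponytail bone chain with head and upper-spine bodies.
# Each bone has one capsule, zero-length (effectively a sphere) except the large spine_05 capsule.
# Locked angular constraints tie the ponytail base to the head and the head to spine_05.
# The four ponytail links below have per-joint swing limits, with twist locked at the first and tightly limited after.
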PhysicsAsset /BRCosmetics/Characters/Player/Female/Medium/Bodies/F_MED_AquaPeony/Meshes/Parts/F_MED_AquaPeony_FaceAcc_Physics
bodies {
 dyn_ponytail_00 [{
   shapeType=Capsule
   length=0
   radius=3.7976
   shapeName=dyn_ponytail_00_capsule
   center=<4.2153 0.7224 1.2163>
   rotation=<rot -22.2725 6.7397 -86.8182>
  }]
 dyn_ponytail_01 [{shapeType=Capsule, length=0, radius=4.3623, shapeName=dyn_ponytail_01_capsule, center=<2.1701 -0.2336 0.3427>}]
 dyn_ponytail_02 [{shapeType=Capsule, length=0, radius=3.8664, shapeName=dyn_ponytail_02_capsule, center=<2.7177 -0.3403 0.4473>, rotation=<rot 14.1081 0 0>}]
 dyn_ponytail_03 [{shapeType=Capsule, length=0, radius=3.4847, shapeName=dyn_ponytail_03_capsule, center=<3.1892 0.1774 1.8324>}]
 dyn_ponytail_base [{shapeType=Capsule, length=0, radius=0.505, shapeName=dyn_ponytail_base_capsule}]
 head [{shapeType=Capsule, length=0, radius=9.4125, center=<6.0604 0.6548 -0.1076>}]
 spine_05 [{shapeType=Capsule, length=15, radius=15, shapeName=spine_05_capsule, center=<-2.5962 2.2624 0>}]
}
constraints [
 {
  bone1Name=dyn_ponytail_base
  bone2Name=head
  swing1LimitDegrees=45
  swing1Motion=Locked
  swing2LimitDegrees=45
  swing2Motion=Locked
  twistLimitDegrees=45
  twistMotion=Locked
 }
 {
  bone1Name=dyn_ponytail_00
  bone2Name=dyn_ponytail_base
  swing1LimitDegrees=10
  swing2LimitDegrees=20
  twistLimitDegrees=0
  twistMotion=Locked
 }
 {bone1Name=dyn_ponytail_01, bone2Name=dyn_ponytail_00, swing1LimitDegrees=20, swing2LimitDegrees=30, twistLimitDegrees=2}
 {bone1Name=dyn_ponytail_02, bone2Name=dyn_ponytail_01, swing1LimitDegrees=30, swing2LimitDegrees=20, twistLimitDegrees=2}
 {bone1Name=dyn_ponytail_03, bone2Name=dyn_ponytail_02, swing1LimitDegrees=30, swing2LimitDegrees=5, twistLimitDegrees=2}
 {
  bone1Name=head
  bone2Name=spine_05
  swing1LimitDegrees=45
  swing1Motion=Locked
  swing2LimitDegrees=45
  swing2Motion=Locked
  twistLimitDegrees=45
  twistMotion=Locked
 }
]
