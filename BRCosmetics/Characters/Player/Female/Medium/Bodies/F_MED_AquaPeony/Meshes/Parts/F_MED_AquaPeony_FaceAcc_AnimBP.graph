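# Post-copy AnimBP for the AquaPeony female face accessory, adding secondary motion to hair on the pose copied from the parent mesh.
# AnimDynamics chains move bangs, side pieces and strands, then bRigidBodyEnabled selects RigidBody simulation or an AnimDynamics ponytail fallback.
# Additive bone rotations lift the ponytail during skydiving and floating, before a Control Rig receives the backpack flag.
# The EventGraph updates RigidBody gravity for floating and detects backpacks versus capes.
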AnimBlueprint /BRCosmetics/Characters/Player/Female/Medium/Bodies/F_MED_AquaPeony/Meshes/Parts/F_MED_AquaPeony_FaceAcc_AnimBP
target_skeleton=&/F_MED_AquaPeony_FaceAcc_Skeleton
graphs [
 AnimGraph {nodes [
   n1 = Anim Dynamics - Bone: dyn_ponytail_00(ComponentPose=Local To Component(LocalPose=Use cached pose 'dynamic pose'()), Alpha=1, ExternalForce="0,0,0") {
    angularDampingOverride=0.9999
    bOverrideAngularDamping=true
    bUseGravityOverride=true
    alphaBoolBlend {blendOption=Linear}
    boundBone {boneName=dyn_ponytail_00}
    componentAppliedLinearAccClamp=<1000 1500 1000>
    componentLinearAccScale=<0.4 0.5 0.4>
    componentLinearVelScale=<0.5 0.6 0.5>
    gravityOverride {z=-800}
    physicsBodyDefinitions [{
      boundBone {boneName=dyn_ponytail_00}
      boxExtents=<7 7 7>
      constraintSetup {angularLimitsMax=<5 0 30>, angularLimitsMin=<-5 -30 -30>, angularTarget=<0 0 0>}
      localJointOffset=<1 0 0>
     }]
   }
   n2 = Anim Dynamics - Bone: dyn_ponytail_01(ComponentPose=@n1, Alpha=1, ExternalForce="0,0,0") {
    angularDampingOverride=0.9999
    bOverrideAngularDamping=true
    bUseGravityOverride=true
    alphaBoolBlend {blendOption=Linear}
    boundBone {boneName=dyn_ponytail_01}
    componentAppliedLinearAccClamp=<1000 1500 1000>
    componentLinearAccScale=<0.4 0.5 0.4>
    componentLinearVelScale=<0.5 0.6 0.5>
    gravityOverride {z=-800}
    physicsBodyDefinitions [{
      boundBone {boneName=dyn_ponytail_01}
      boxExtents=<7 7 7>
      constraintSetup {angularLimitsMax=<5 0 30>, angularLimitsMin=<-5 -40 -30>, angularTarget=<0 0 0>}
      localJointOffset=<1 0 0>
     }]
   }
   n3 = Anim Dynamics - Bone: dyn_ponytail_02(ComponentPose=@n2, Alpha=1, ExternalForce="0,0,0") {
    angularDampingOverride=0.9999
    bOverrideAngularDamping=true
    bUseGravityOverride=true
    bUseSphericalLimits=true
    alphaBoolBlend {blendOption=Linear}
    boundBone {boneName=dyn_ponytail_02}
    componentAppliedLinearAccClamp=<1000 1500 1000>
    componentLinearAccScale=<0.4 0.5 0.4>
    componentLinearVelScale=<0.5 0.6 0.5>
    gravityOverride {z=-800}
    physicsBodyDefinitions [{
      boundBone {boneName=dyn_ponytail_02}
      boxExtents=<7 7 7>
      constraintSetup {angularLimitsMax=<5 30 10>, angularLimitsMin=<-5 0 -20>, angularTarget=<0 0 0>}
      localJointOffset=<1 0 0>
     }]
    sphericalLimits [{limitRadius=10, drivingBone {boneName=head}, sphereLocalOffset=<4.7113 0.4104 0>}]
   }
   n4 = Anim Dynamics - Bone: dyn_ponytail_03(ComponentPose=@n3, Alpha=1, ExternalForce="0,0,0") {
    angularDampingOverride=0.9999
    bOverrideAngularDamping=true
    bUseGravityOverride=true
    bUseSphericalLimits=true
    alphaBoolBlend {blendOption=Linear}
    boundBone {boneName=dyn_ponytail_03}
    componentAppliedLinearAccClamp=<1000 1500 1000>
    componentLinearAccScale=<0.4 0.5 0.4>
    componentLinearVelScale=<0.5 0.6 0.5>
    gravityOverride {z=-800}
    physicsBodyDefinitions [{
      boundBone {boneName=dyn_ponytail_03}
      boxExtents=<7 7 7>
      constraintSetup {angularLimitsMax=<5 15 10>, angularLimitsMin=<-5 0 -20>, angularTarget=<0 0 0>}
      localJointOffset=<1 0 0>
     }]
    sphericalLimits [
     {limitRadius=10, drivingBone {boneName=head}, sphereLocalOffset=<6.2776 0 0>}
     {limitRadius=14.5, drivingBone {boneName=spine_05}, sphereLocalOffset=<2.491 1.6471 0>}
     {limitRadius=5, drivingBone {boneName=spine_05}, sphereLocalOffset=<0 0 0>}
    ]
   }
   n5 = Get bIsSkydiveFloating()
   n6 = "Transform (Modify) Bone - Bone: dyn_ponytail_00"!(ComponentPose=@n4, bAlphaBoolEnabled:bool=@n5, Translation="0,0,0", Rotation="40,0,0", Scale="1,1,1") -> (Pose) {
    alphaInputType=Bool
    rotationMode=BMM_Additive
    rotationSpace=BCS_BoneSpace
    translationMode=BMM_Additive
    translationSpace=BCS_BoneSpace
    alphaBoolBlend {blendInTime=0.2, blendOption=Linear, blendOutTime=0.2}
    boneToModify {boneName=dyn_ponytail_00}
    rotation {pitch=40}
   }
   n7 = "Transform (Modify) Bone - Bone: dyn_ponytail_01"!(ComponentPose=@n6, bAlphaBoolEnabled:bool=@n5, Translation="0,0,0", Rotation="35.691,-0.000,-0.000", Scale="1,1,1") -> (Pose) {
    alphaInputType=Bool
    rotationMode=BMM_Additive
    rotationSpace=BCS_BoneSpace
    translationMode=BMM_Additive
    translationSpace=BCS_BoneSpace
    alphaBoolBlend {blendInTime=0.2, blendOption=Linear, blendOutTime=0.2}
    boneToModify {boneName=dyn_ponytail_01}
    rotation=<rot 35.691 0 0>
   }
   n8 = "Transform (Modify) Bone - Bone: dyn_ponytail_02"!(ComponentPose=@n7, bAlphaBoolEnabled:bool=@n5, Translation="0,0,0", Rotation="27.586,0.000,-0.000", Scale="1,1,1") -> (Pose) {
    alphaInputType=Bool
    rotationMode=BMM_Additive
    rotationSpace=BCS_BoneSpace
    translationMode=BMM_Additive
    translationSpace=BCS_BoneSpace
    alphaBoolBlend {blendInTime=0.2, blendOption=Linear, blendOutTime=0.2}
    boneToModify {boneName=dyn_ponytail_02}
    rotation=<rot 27.586 0 0>
   }
   n9 = "Transform (Modify) Bone - Bone: dyn_ponytail_03"!(ComponentPose=@n8, bAlphaBoolEnabled:bool=@n5, Translation="0,0,0", Rotation="5.118,0.000,-0.000", Scale="1,1,1") -> (Pose) {
    alphaInputType=Bool
    rotationMode=BMM_Additive
    rotationSpace=BCS_BoneSpace
    translationMode=BMM_Additive
    translationSpace=BCS_BoneSpace
    alphaBoolBlend {blendInTime=0.2, blendOption=Linear, blendOutTime=0.2}
    boneToModify {boneName=dyn_ponytail_03}
    rotation=<rot 5.118 0 0>
   }
   n10 = Component To Local(ComponentPose=@n9)
   n11 = RigidBody!(
    ComponentPose=Local To Component(LocalPose=Use cached pose 'dynamic pose'())
    Alpha
    OverrideWorldGravity:Vector=Get Gravity Override()
    ExternalForce
   ) -> (Pose) {
    bOverrideWorldGravity=true
    overridePhysicsAsset=&/F_MED_AquaPeony_FaceAcc_Physics
    simulationSpace=BaseBoneSpace
    alphaBoolBlend {blendOption=Linear}
    baseBoneRef {boneName=spine_05}
    simSpaceSettings {
     maxAngularAcceleration=1
     maxAngularVelocity=1
     maxLinearAcceleration=1200
     maxLinearVelocity=700
     velocityScaleZ=0.35
     worldAlpha=1
     externalLinearDragV=<0.5 0.5 0.5>
    }
   }
   n12 = Component To Local(ComponentPose=@n11)
   n13 = Get bRigidBodyEnabled()
   n14 = "Transform (Modify) Bone - Bone: dyn_ponytail_00"(
    ComponentPose=Local To Component(LocalPose=Blend Poses by bool(BlendPose_0=@n12, BlendPose_1=@n10, bActiveValue=@n13) {blendType=Linear})
    bAlphaBoolEnabled=Get bIsSkydiveFloating()
    Translation="0,0,0"
    Rotation="7,20,0"
    Scale="1,1,1"
   ) {
    alphaInputType=Bool
    rotationMode=BMM_Additive
    rotationSpace=BCS_BoneSpace
    translationMode=BMM_Additive
    translationSpace=BCS_BoneSpace
    alphaBoolBlend {blendInTime=0.2, blendOption=Linear, blendOutTime=0.2}
    boneToModify {boneName=dyn_ponytail_00}
    rotation {pitch=7, yaw=20}
   }
   n15 = "Transform (Modify) Bone - Bone: dyn_ponytail_00"(ComponentPose=@n14, bAlphaBoolEnabled=Get bIsSkyDiving(), Translation="0,0,0", Rotation="25,0,0", Scale="1,1,1") {
    alphaInputType=Bool
    rotationMode=BMM_Additive
    rotationSpace=BCS_BoneSpace
    translationMode=BMM_Additive
    translationSpace=BCS_BoneSpace
    alphaBoolBlend {blendInTime=0.2, blendOption=Linear, blendOutTime=0.2}
    boneToModify {boneName=dyn_ponytail_00}
    rotation {pitch=25}
   }
   n16 = Control Rig!(Source=Component To Local(ComponentPose=@n15), Alpha, isWearingBackpack:double=Get IsWearingBackpack() -> (IsWearingBackpack:double)) -> (Pose) {
    alphaBoolBlend {blendOption=Linear}
    controlRigAssetReference {blueprintRigClass=&/F_MED_AquaPeony_FaceAcc_CtrlRig.F_MED_AquaPeony_FaceAcc_CtrlRig_C}
   }
   Output Pose(Result=@n16) {layerGroup=None}
   Save cached pose 'input pose'(Pose=Copy Pose From Mesh() {bUseAttachedParent=true}) {cacheName=input pose}
   n17 = Anim Dynamics - Bone: dyn_bangs_A_00(ComponentPose=Local To Component(LocalPose=Use cached pose 'input pose'()), Alpha=1, ExternalForce="0,0,0") {
    angularDampingOverride=0.9999
    bOverrideAngularDamping=true
    bUseGravityOverride=true
    alphaBoolBlend {blendOption=Linear}
    boundBone {boneName=dyn_bangs_A_00}
    componentAppliedLinearAccClamp=<800 800 800>
    componentLinearAccScale=<0.3 0.3 0.3>
    componentLinearVelScale=<0.4 0.4 0.4>
    gravityOverride {z=-980}
    physicsBodyDefinitions [{
      boundBone {boneName=dyn_bangs_A_00}
      boxExtents=<10 7 7>
      constraintSetup {angularLimitsMax=<10 5 20>, angularLimitsMin=<-5 -15 0>, angularTarget=<0 0 0>}
      localJointOffset=<1 0 0>
     }]
   }
   n18 = Anim Dynamics - Bone: dyn_bangs_A_01(ComponentPose=@n17, Alpha=1, ExternalForce="0,0,0") {
    angularDampingOverride=0.9999
    bOverrideAngularDamping=true
    bUseGravityOverride=true
    alphaBoolBlend {blendOption=Linear}
    boundBone {boneName=dyn_bangs_A_01}
    componentAppliedLinearAccClamp=<800 800 800>
    componentLinearAccScale=<0.4 0.4 0.4>
    componentLinearVelScale=<0.5 0.5 0.5>
    gravityOverride {z=-850}
    physicsBodyDefinitions [{
      boundBone {boneName=dyn_bangs_A_01}
      boxExtents=<7 7 7>
      constraintSetup {angularLimitsMax=<5 15 0>, angularLimitsMin=<-5 -10 -20>, angularTarget=<0 0 0>}
      localJointOffset=<1 0 0>
     }]
   }
   n19 = Anim Dynamics - Bone: dyn_bangs_B_00(ComponentPose=@n18, Alpha=1, ExternalForce="0,0,0") {
    angularDampingOverride=0.9999
    bOverrideAngularDamping=true
    bUseGravityOverride=true
    alphaBoolBlend {blendOption=Linear}
    boundBone {boneName=dyn_bangs_B_00}
    componentAppliedLinearAccClamp=<800 800 800>
    componentLinearAccScale=<0.3 0.3 0.3>
    componentLinearVelScale=<0.4 0.4 0.4>
    gravityOverride {z=-980}
    physicsBodyDefinitions [{
      boundBone {boneName=dyn_bangs_B_00}
      boxExtents=<10 7 7>
      constraintSetup {angularLimitsMax=<0 2 0>, angularLimitsMin=<0 -15 -20>, angularTarget=<0 0 0>}
      localJointOffset=<1 0 0>
     }]
   }
   n20 = Anim Dynamics - Bone: dyn_bangs_C_00(ComponentPose=@n19, Alpha=1, ExternalForce="0,0,0") {
    angularDampingOverride=0.9999
    bOverrideAngularDamping=true
    bUseGravityOverride=true
    alphaBoolBlend {blendOption=Linear}
    boundBone {boneName=dyn_bangs_C_00}
    componentAppliedLinearAccClamp=<800 800 800>
    componentLinearAccScale=<0.3 0.3 0.3>
    componentLinearVelScale=<0.4 0.4 0.4>
    gravityOverride {z=-980}
    physicsBodyDefinitions [{
      boundBone {boneName=dyn_bangs_C_00}
      boxExtents=<10 7 7>
      constraintSetup {angularLimitsMax=<0 5 0>, angularLimitsMin=<0 -15 -20>, angularTarget=<0 0 0>}
      localJointOffset=<1 0 0>
     }]
   }
   n21 = Anim Dynamics - Bone: dyn_bangs_C_01(ComponentPose=@n20, Alpha=1, ExternalForce="0,0,0") {
    angularDampingOverride=0.9999
    bOverrideAngularDamping=true
    bUseGravityOverride=true
    alphaBoolBlend {blendOption=Linear}
    boundBone {boneName=dyn_bangs_C_01}
    componentAppliedLinearAccClamp=<800 800 800>
    componentLinearAccScale=<0.4 0.4 0.4>
    componentLinearVelScale=<0.5 0.5 0.5>
    gravityOverride {z=-800}
    physicsBodyDefinitions [{
      boundBone {boneName=dyn_bangs_C_01}
      boxExtents=<7 7 7>
      constraintSetup {angularLimitsMax=<5 30 0>, angularLimitsMin=<-5 -5 -10>, angularTarget=<0 0 0>}
      localJointOffset=<1 0 0>
     }]
   }
   n22 = Anim Dynamics - Bone: dyn_bangs_D_00(ComponentPose=@n21, Alpha=1, ExternalForce="0,0,0") {
    angularDampingOverride=0.9999
    bOverrideAngularDamping=true
    bUseGravityOverride=true
    alphaBoolBlend {blendOption=Linear}
    boundBone {boneName=dyn_bangs_D_00}
    componentAppliedLinearAccClamp=<800 800 800>
    componentLinearAccScale=<0.3 0.3 0.3>
    componentLinearVelScale=<0.4 0.4 0.4>
    gravityOverride {z=-980}
    physicsBodyDefinitions [{
      boundBone {boneName=dyn_bangs_D_00}
      boxExtents=<10 10 10>
      constraintSetup {angularLimitsMax=<10 0 5>, angularLimitsMin=<-10 -15 -20>, angularTarget=<0 0 0>}
      localJointOffset=<1 0 0>
     }]
   }
   n23 = Anim Dynamics - Bone: dyn_bangs_D_01(ComponentPose=@n22, Alpha=1, ExternalForce="0,0,0") {
    angularDampingOverride=0.9999
    bOverrideAngularDamping=true
    bUseGravityOverride=true
    alphaBoolBlend {blendOption=Linear}
    boundBone {boneName=dyn_bangs_D_01}
    componentAppliedLinearAccClamp=<1000 1000 1000>
    componentLinearAccScale=<0.4 0.4 0.4>
    componentLinearVelScale=<0.5 0.5 0.5>
    gravityOverride {z=-800}
    physicsBodyDefinitions [{
      boundBone {boneName=dyn_bangs_D_01}
      boxExtents=<7 7 7>
      constraintSetup {angularLimitsMax=<5 0 10>, angularLimitsMin=<-5 -5 -20>, angularTarget=<0 0 0>}
      localJointOffset=<1 0 0>
     }]
   }
   n24 = Anim Dynamics - Bone: dyn_L_sidepiece_00(ComponentPose=@n23, Alpha=1, ExternalForce="0,0,0") {
    angularDampingOverride=0.9999
    bOverrideAngularDamping=true
    bUseGravityOverride=true
    bUseSphericalLimits=true
    alphaBoolBlend {blendOption=Linear}
    boundBone {boneName=dyn_L_sidepiece_00}
    componentAppliedLinearAccClamp=<1000 1000 1000>
    componentLinearAccScale=<0.4 0.4 0.4>
    componentLinearVelScale=<0.5 0.5 0.5>
    gravityOverride {z=-800}
    physicsBodyDefinitions [{
      boundBone {boneName=dyn_L_sidepiece_00}
      boxExtents=<10 10 10>
      constraintSetup {angularLimitsMax=<10 -4 20>, angularLimitsMin=<-10 -15 -20>, angularTarget=<0 0 0>}
      localJointOffset=<1 0 0>
     }]
    sphericalLimits [{limitRadius=9.5, drivingBone {boneName=head}, sphereLocalOffset=<1.4829 2.6794 1.7046>}]
   }
   n25 = Anim Dynamics - Bone: dyn_L_sidepiece_01(ComponentPose=@n24, Alpha=1, ExternalForce="0,0,0") {
    angularDampingOverride=0.9999
    bOverrideAngularDamping=true
    bUseGravityOverride=true
    bUsePlanarLimit=false
    bUseSphericalLimits=true
    alphaBoolBlend {blendOption=Linear}
    boundBone {boneName=dyn_L_sidepiece_01}
    componentAppliedLinearAccClamp=<1000 1000 1000>
    componentLinearAccScale=<0.4 0.4 0.4>
    componentLinearVelScale=<0.5 0.5 0.5>
    gravityOverride {z=-700}
    physicsBodyDefinitions [{
      boundBone {boneName=dyn_L_sidepiece_01}
      boxExtents=<10 10 10>
      constraintSetup {angularLimitsMax=<10 0 20>, angularLimitsMin=<-10 -15 -20>, angularTarget=<0 0 0>}
      localJointOffset=<1 0 0>
     }]
    sphericalLimits [{limitRadius=8, drivingBone {boneName=head}, sphereLocalOffset=<0.3029 4.923 0.4778>}]
   }
   n26 = Anim Dynamics - Bone: dyn_R_sidepiece_00(ComponentPose=@n25, Alpha=1, ExternalForce="0,0,0") {
    angularDampingOverride=1
    bOverrideAngularDamping=true
    bUseGravityOverride=true
    alphaBoolBlend {blendOption=Linear}
    boundBone {boneName=dyn_R_sidepiece_00}
    componentAppliedLinearAccClamp=<1000 1000 1000>
    componentLinearAccScale=<0.5 0.5 0.5>
    componentLinearVelScale=<0.6 0.6 0.6>
    gravityOverride {z=-700}
    physicsBodyDefinitions [{
      boundBone {boneName=dyn_R_sidepiece_00}
      boxExtents=<10 10 10>
      constraintSetup {angularLimitsMax=<5 20 10>, angularLimitsMin=<-5 -40 -20>, angularTarget=<0 0 0>}
      localJointOffset=<1 0 0>
     }]
   }
   n27 = Anim Dynamics - Bone: dyn_R_sidepiece_01(ComponentPose=@n26, Alpha=1, ExternalForce="0,0,0") {
    angularDampingOverride=1
    bOverrideAngularDamping=true
    bUseGravityOverride=true
    bUseSphericalLimits=true
    alphaBoolBlend {blendOption=Linear}
    boundBone {boneName=dyn_R_sidepiece_01}
    componentAppliedLinearAccClamp=<1000 1000 1000>
    componentLinearAccScale=<0.4 0.4 0.4>
    componentLinearVelScale=<0.5 0.5 0.5>
    gravityOverride {z=-700}
    physicsBodyDefinitions [{
      boundBone {boneName=dyn_R_sidepiece_01}
      boxExtents=<10 10 10>
      constraintSetup {angularLimitsMax=<5 10 20>, angularLimitsMin=<-5 -15 -20>, angularTarget=<0 0 0>}
      localJointOffset=<1.2 0 0>
     }]
    sphericalLimits [{limitRadius=6.5, drivingBone {boneName=head}, sphereLocalOffset=<0 0 0.9708>}]
   }
   n28 = Anim Dynamics - Bone: dyn_strand_E(ComponentPose=@n27, Alpha=1, ExternalForce="0,0,0") {
    angularDampingOverride=1
    bOverrideAngularDamping=true
    bUseGravityOverride=true
    alphaBoolBlend {blendOption=Linear}
    boundBone {boneName=dyn_strand_E}
    componentAppliedLinearAccClamp=<1000 1000 1000>
    componentLinearAccScale=<0.5 0.5 0.5>
    componentLinearVelScale=<0.6 0.6 0.6>
    gravityOverride {z=-700}
    physicsBodyDefinitions [{
      boundBone {boneName=dyn_strand_E}
      boxExtents=<5 5 5>
      constraintSetup {angularLimitsMax=<0 15 20>, angularLimitsMin=<0 -15 -20>, angularTarget=<0 0 0>}
      localJointOffset=<1 0 0>
     }]
   }
   n29 = Anim Dynamics - Bone: dyn_strand_C(ComponentPose=@n28, Alpha=1, ExternalForce="0,0,0") {
    angularDampingOverride=1
    bOverrideAngularDamping=true
    bUseGravityOverride=true
    alphaBoolBlend {blendOption=Linear}
    boundBone {boneName=dyn_strand_C}
    componentAppliedLinearAccClamp=<1000 1000 1000>
    componentLinearAccScale=<0.5 0.5 0.5>
    componentLinearVelScale=<0.6 0.6 0.6>
    gravityOverride {z=-700}
    physicsBodyDefinitions [{
      boundBone {boneName=dyn_strand_C}
      boxExtents=<5 5 5>
      constraintSetup {angularLimitsMax=<10 15 20>, angularLimitsMin=<-10 -15 -20>, angularTarget=<0 0 0>}
      localJointOffset=<1 0 0>
     }]
   }
   n30 = Anim Dynamics - Bone: dyn_strand_B(ComponentPose=@n29, Alpha=1, ExternalForce="0,0,0") {
    angularDampingOverride=1
    bOverrideAngularDamping=true
    bUseGravityOverride=true
    alphaBoolBlend {blendOption=Linear}
    boundBone {boneName=dyn_strand_B}
    componentAppliedLinearAccClamp=<1000 1000 1000>
    componentLinearAccScale=<0.5 0.5 0.5>
    componentLinearVelScale=<0.6 0.6 0.6>
    gravityOverride {z=-700}
    physicsBodyDefinitions [{
      boundBone {boneName=dyn_strand_B}
      boxExtents=<5 5 5>
      constraintSetup {angularLimitsMax=<10 15 20>, angularLimitsMin=<-10 -15 -20>, angularTarget=<0 0 0>}
      localJointOffset=<1 0 0>
     }]
   }
   n31 = Anim Dynamics - Bone: dyn_strand_A(ComponentPose=@n30, Alpha=1, ExternalForce="0,0,0") {
    angularDampingOverride=1
    bOverrideAngularDamping=true
    bUseGravityOverride=true
    alphaBoolBlend {blendOption=Linear}
    boundBone {boneName=dyn_strand_A}
    componentAppliedLinearAccClamp=<1000 1000 1000>
    componentLinearAccScale=<0.5 0.5 0.5>
    componentLinearVelScale=<0.6 0.6 0.6>
    gravityOverride {z=-700}
    physicsBodyDefinitions [{
      boundBone {boneName=dyn_strand_A}
      boxExtents=<5 5 5>
      constraintSetup {angularLimitsMax=<10 15 20>, angularLimitsMin=<-10 -15 -20>, angularTarget=<0 0 0>}
      localJointOffset=<1 0 0>
     }]
   }
   Save cached pose 'dynamic pose'(Pose=Component To Local(ComponentPose=@n31)) {cacheName=dynamic pose}
  ]}
 EventGraph {nodes [
   Event BlueprintUpdateAnimation()
   > n1 = Sequence()
   > n2 = Branch(Condition=Get bIsSkydiveFloating())
   > Set Gravity Override(Gravity Override="0,-700,600")
   Set Gravity Override(execute=@n2.else, Gravity Override="0,0,-980")
   n3 = CheckForBackbling()
   n4 = NOT Boolean(A=@n3.bIsBackblingACape)
   Set IsWearingBackpack(execute=@n1.then_1, IsWearingBackpack:double="To Float (Boolean)"(InBool=AND Boolean(A=@n3, B=@n4))) -> (Output_Get:double)
   > Set IsWearingCape(IsWearingCape:double="To Float (Boolean)"(InBool=CheckForBackbling().bIsBackblingACape)) -> (Output_Get:double)
   > Set IsWearingBackbling(IsWearingBackbling:double=float + float(A=Get IsWearingBackbling() -> (IsWearingBackbling:double), B=Get IsWearingCape() -> (IsWearingCape:double))) -> (Output_Get:double)
  ]}
]
variables [
 Gravity Override {type=Vector}
 Gravity Override Params {type=GravityOverrideParamsStruct}
 IsWearingBackbling {type=number}
 IsWearingBackpack {type=number}
 IsWearingCape {type=number}
 NewVar {type=number}
]
